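# Autumn vehicle boost cosmetic effect that sprays spinning orange, yellow and red sprites backward from skeletal mesh sockets.
# Four emitters share one pattern: filtered socket count partitions particles across ports, then cone velocity and sphere offset scatter them.
# Two slower-rate emitters decelerate by curve, one adding a ribbon renderer; two denser ones use curve-driven randomized gravity and SubUV animation.
# All drive dynamic material parameters, while painted-variant overrides are disabled.
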NiagaraSystem /VehicleCosmetics/Boosts/Boost_Autumn/FX/NS_Boost_Autumn
dependencies {
 usedDataInterfaces [
  &NiagaraDataInterfaceCurve
  &NiagaraDataInterfaceSpriteRendererInfo
  &NiagaraDataInterfaceSkeletalMesh
 ]
 usedDynamicInputs []
 usedModules [
  &self:SystemSpawnScript.SystemScriptSource.SystemScriptGraph.NiagaraNodeAssignment_0.SetVariables_#1
  &SystemState
 ]
}
emitters {
 NE_VC_Boost_Sprite_Parent {
  emitterUpdateScript [
   &EmitterState()
   &VC_FilteredSocketCount()
   &SpawnRate(SpawnRate=30, Use Spawn Probability=false)
  ]
  particleSpawnScript [
   &VC_PaintedVariantsOverride() {enabled=false}
   &InitializeParticle(
    Color Mode=::NewEnumerator0
    Lifetime Max=1.5
    Lifetime Min=0.8
    Lifetime Mode=::NewEnumerator1
    Sprite Rotation Mode=::NewEnumerator2
    Sprite Size Mode=::NewEnumerator3
    Uniform Sprite Size=45
    UsePositionOffset=false
    Write Material Random=true
   )
   &PartitionParticles(Number Of Partitions=$Emitter.PortCount)
   &SkeletalMeshLocation(
    Bone / Socket Index=$Particles.PartitionParticles.Partition
    Bone / Socket Sampling Mode=::NewEnumerator5
    MaskSpawnGroup=false
    Mesh Sampling Type=::NewEnumerator6
    Skeletal Mesh=$System.SkeletalMesh
   )
   &AddVelocity(Velocity=&RandomRangeVector(Minimum=<-355 0 0>, Maximum=<-1500 0 0>)) {enabled=false}
   &AddVelocity(
    Cone Angle=15
    Cone Axis=<-10 0 0>
    Module.UseConeFalloff=false
    Module.UseVelDistribution=false
    Override Seed=false
    Velocity Mode=::NewEnumerator2
    Velocity Speed=1500
   ) {module=AddVelocity001}
   &ShapeLocation(MaskSpawnGroup=false, Override Seed=false, Sphere Radius=15, UseOwnerScale=true)
   &Color(Color=&RandomRangeLinearColor(Minimum=<rgba 0.8 0.5 0.05 1>, Maximum=<rgba 0.6812 0.292 0.0487 1>))
  ]
  particleUpdateScript [
   &ParticleState()
   &ScaleVelocity(
    Coordinate Space:ENiagaraCoordinateSpace=::World
    Velocity Scale=&VectorFromFloat(Value=&FloatFromCurve(FloatCurve=DataInterface() {propertyValues {Curve {keys [<at 0 0.5 RCIM_Cubic>,<at 1 0.1 RCIM_Cubic>]}}}))
   )
   &ScaleSpriteSize(Uniform Curve Sprite Scale=DataInterface() {propertyValues {Curve {keys [{tangentMode=RCTM_User, time=0.8, value=1}, {tangentMode=RCTM_User, time=1, value=0}]}}})
   &SpriteRotationRate(Rotation Rate=&RandomRangeFloat(Minimum=0, Maximum=360))
   &SolveForcesAndVelocity(Clamp Velocity=false, Limit Acceleration=false)
   &ScaleColor(Scale RGB=<1.5 1 1>, ScaleA=true, ScaleRGB=true)
   &DynamicMaterialParameters(
    Index 0 Param 1=&RandomRangeFloat(Minimum=0.75, Maximum=0.9)
    Index 0 Param 2=1
    Index 0 Param 3=&RandomRangeFloat(Minimum=0.06, Maximum=0.1)
    Index 0 Param 4=&RandomRangeFloat(Minimum=0.1, Maximum=0.2)
    Module.Index0Param0Write=true
    Module.Index0Param1Write=true
    Module.Index0Param2Write=true
    Module.Index0Param3Write=true
   )
  ]
  rendererClasses [&NiagaraRibbonRendererProperties, &NiagaraSpriteRendererProperties]
  renderers [
   {rendererClass=&NiagaraRibbonRendererProperties}
   {rendererClass=&NiagaraSpriteRendererProperties, rendererIndex=1}
  ]
 }
 NE_VC_Boost_Sprite_Parent001 {
  emitterUpdateScript [
   &EmitterState()
   &VC_FilteredSocketCount()
   &SpawnRate(SpawnRate=30, Use Spawn Probability=false)
  ]
  particleSpawnScript [
   &VC_PaintedVariantsOverride() {enabled=false}
   &InitializeParticle(
    Color Mode=::NewEnumerator0
    Lifetime Max=1.5
    Lifetime Min=0.8
    Lifetime Mode=::NewEnumerator1
    Sprite Rotation Mode=::NewEnumerator2
    Sprite Size Mode=::NewEnumerator3
    Uniform Sprite Size=45
    UsePositionOffset=false
    Write Material Random=true
   )
   &PartitionParticles(Number Of Partitions=$Emitter.PortCount)
   &SkeletalMeshLocation(
    Bone / Socket Index=$Particles.PartitionParticles.Partition
    Bone / Socket Sampling Mode=::NewEnumerator5
    MaskSpawnGroup=false
    Mesh Sampling Type=::NewEnumerator6
    Skeletal Mesh=$System.SkeletalMesh
   )
   &AddVelocity(Velocity=&RandomRangeVector(Minimum=<-355 0 0>, Maximum=<-1500 0 0>)) {enabled=false}
   &AddVelocity(
    Cone Angle=15
    Cone Axis=<-10 0 0>
    Module.UseConeFalloff=false
    Module.UseVelDistribution=false
    Override Seed=false
    Velocity Mode=::NewEnumerator2
    Velocity Speed=1500
   ) {module=AddVelocity001}
   &ShapeLocation(MaskSpawnGroup=false, Override Seed=false, Sphere Radius=15, UseOwnerScale=true)
   &Color(Color=&RandomRangeLinearColor(Minimum=<rgba 0.8 0.5 0.05 1>, Maximum=<rgba 0.7 0.3 0.05 1>))
  ]
  particleUpdateScript [
   &ParticleState()
   &ScaleVelocity(
    Coordinate Space:ENiagaraCoordinateSpace=::World
    Velocity Scale=&VectorFromFloat(Value=&FloatFromCurve(FloatCurve=DataInterface() {propertyValues {Curve {keys [<at 0 0.5 RCIM_Cubic>,<at 1 0.2 RCIM_Cubic>]}}}))
   ) {module=ScaleVelocity001}
   &ScaleSpriteSize(Uniform Curve Sprite Scale=DataInterface() {propertyValues {Curve {keys [{tangentMode=RCTM_User, time=0.8, value=1}, {tangentMode=RCTM_User, time=1, value=0}]}}})
   &SpriteRotationRate(Rotation Rate=&RandomRangeFloat(Minimum=0, Maximum=360))
   &SolveForcesAndVelocity(Clamp Velocity=false, Limit Acceleration=false)
   &ScaleColor(Scale RGB=<1.5 1 1>, ScaleA=true, ScaleRGB=true)
   &DynamicMaterialParameters(
    Index 0 Param 1=&RandomRangeFloat(Minimum=0.75, Maximum=1.5)
    Index 0 Param 2=1
    Index 0 Param 3=&RandomRangeFloat(Minimum=0.05, Maximum=0.1)
    Index 0 Param 4=&RandomRangeFloat(Minimum=0, Maximum=0.1)
    Module.Index0Param0Write=true
    Module.Index0Param1Write=true
    Module.Index0Param2Write=true
    Module.Index0Param3Write=true
   ) {module=DynamicMaterialParameters001}
  ]
  rendererClasses [&NiagaraSpriteRendererProperties]
  renderers [{rendererClass=&NiagaraSpriteRendererProperties}]
 }
 NE_VC_Boost_Sprite_Parent002 {
  emitterUpdateScript [
   &EmitterState()
   &VC_FilteredSocketCount()
   &SpawnRate(SpawnRate=80, Use Spawn Probability=false)
  ]
  particleSpawnScript [
   &VC_PaintedVariantsOverride() {enabled=false}
   &InitializeParticle(
    Color Mode=::NewEnumerator0
    Lifetime Mode=::NewEnumerator1
    Sprite Rotation Mode=::NewEnumerator2
    Sprite Size Mode=::NewEnumerator1
    Uniform Sprite Size Max=45
    Uniform Sprite Size Min=65
    UsePositionOffset=false
    Write Material Random=true
   )
   &PartitionParticles(Number Of Partitions=$Emitter.PortCount)
   &SkeletalMeshLocation(
    Bone / Socket Index=$Particles.PartitionParticles.Partition
    Bone / Socket Sampling Mode=::NewEnumerator5
    MaskSpawnGroup=false
    Mesh Sampling Type=::NewEnumerator6
    Skeletal Mesh=$System.SkeletalMesh
   )
   &AddVelocity(
    Cone Angle=&DistanceBetweenPositions(A=<0 0 20>, B=<0 0 0>)
    Cone Axis=<-1500 0 0>
    Module.UseConeFalloff=false
    Module.UseVelDistribution=false
    Override Seed=false
    Velocity Mode=::NewEnumerator2
    Velocity Speed=1500
   )
   &ShapeLocation(MaskSpawnGroup=false, Override Seed=false, Sphere Radius=60, Transform Order:ENiagara_TransformOrder=::NewEnumerator1, UseOwnerScale=true)
   &Color(Color=&RandomRangeLinearColor(Minimum=<rgba 1 0.5635 0 1>, Maximum=<rgba 1 0.3955 0 1>))
  ]
  particleUpdateScript [
   &ParticleState()
   &ScaleSpriteSize(Uniform Curve Sprite Scale=DataInterface() {propertyValues {Curve {keys [{tangentMode=RCTM_User, time=0.8, value=1}, {tangentMode=RCTM_User, time=1, value=0}]}}})
   &SpriteRotationRate(Rotation Rate=&RandomRangeFloat(Minimum=0.5, Maximum=360))
   &ScaleVelocity(Velocity Scale=&VectorFromFloat(Value=&FloatFromCurve(
      FloatCurve=DataInterface() {propertyValues {Curve {keys [{tangentMode=RCTM_User, time=0, value=1}, {tangentMode=RCTM_User, time=1, value=0}]}}}
      Scale Curve=0.6
     ))) {enabled=false}
   &GravityForce(Gravity=&RandomRangeVector(
     Minimum=&VectorFromFloat(Value=&FloatFromCurve(FloatCurve=DataInterface() {propertyValues {Curve {keys [
           {arriveTangent=26315.7891, interpMode=RCIM_Cubic, time=0, value=1}
           {arriveTangent=-369.8653, interpMode=RCIM_Cubic, leaveTangent=-369.8653, tangentMode=RCTM_User, time=0.15, value=0}
           {arriveTangent=-12000.001, interpMode=RCIM_Cubic, leaveTangent=-12000.001, time=0.19, value=5000}
           {arriveTangent=163.0197, interpMode=RCIM_Cubic, leaveTangent=163.0197, tangentMode=RCTM_User, time=0.2, value=-600}
          ]}}}))
     Maximum=<0 0 -600>
    )) {module=GravityForce003}
   &GravityForce(Gravity=&RandomRangeVector(
     Minimum=<0 0 -600>
     Maximum=&VectorFromFloat(Value=&FloatFromCurve(FloatCurve=DataInterface() {propertyValues {Curve {keys [<at 0 1 RCIM_Cubic>, {arriveTangent=-351, interpMode=RCIM_Cubic, leaveTangent=-351, time=0.5, value=20}, <at 1 -350>]}}}))
    )) {enabled=false}
   &GravityForce(Gravity=&AddVectorToPosition(Position=<0 0 -600>, Vector=<0 0 -350>)) {enabled=false, module=GravityForce001}
   &GravityForce(Gravity=&VectorFromFloat(Value=&FloatFromCurve(FloatCurve=DataInterface() {propertyValues {Curve {keys [{arriveTangent=2.0365, interpMode=RCIM_Cubic, leaveTangent=2.0365, tangentMode=RCTM_User, time=0, value=1}, <at 1 0.8>]}}}))) {module=GravityForce002}
   &Drag(Drag=0.5, Use Linear Drag=true, Use Rotational Drag=true) {enabled=false}
   &SolveForcesAndVelocity(Clamp Velocity=false, Limit Acceleration=false)
   &ScaleColor(Scale RGB=<1.2 1 1>, ScaleA=true, ScaleRGB=true)
   &RotateAroundPoint() {enabled=false}
   &DynamicMaterialParameters(
    Index 0 Param 1=&RandomRangeFloat(Minimum=0.75, Maximum=1)
    Index 0 Param 2=0.3
    Index 0 Param 3=&RandomRangeFloat(Minimum=0.15, Maximum=0.3)
    Index 0 Param 4=&RandomRangeFloat(Minimum=0.25, Maximum=0.35)
    Module.Index0Param0Write=true
    Module.Index0Param1Write=true
    Module.Index0Param2Write=true
    Module.Index0Param3Write=true
   ) {module=DynamicMaterialParameters001}
   &SubUVAnimation(
    "Sprite Renderer "=DataInterface() {propertyValues {SpriteRenderer=&self:NE_VC_Boost_Sprite_Parent002.NiagaraSpriteRendererProperties_0}}
    SubUV Lookup Index Scale=0.1
    UseEndFrame=false
    UseStartFrame=false
   )
  ]
  rendererClasses [&NiagaraSpriteRendererProperties]
  renderers [{rendererClass=&NiagaraSpriteRendererProperties}]
 }
 NE_VC_Boost_Sprite_Parent003 {
  emitterUpdateScript [
   &EmitterState()
   &VC_FilteredSocketCount()
   &SpawnRate(SpawnRate=80, Use Spawn Probability=false)
  ]
  particleSpawnScript [
   &VC_PaintedVariantsOverride() {enabled=false}
   &InitializeParticle(
    Color Mode=::NewEnumerator0
    Lifetime Mode=::NewEnumerator1
    Sprite Rotation Mode=::NewEnumerator2
    Sprite Size Mode=::NewEnumerator1
    Uniform Sprite Size Max=45
    Uniform Sprite Size Min=65
    UsePositionOffset=false
    Write Material Random=true
   )
   &PartitionParticles(Number Of Partitions=$Emitter.PortCount)
   &SkeletalMeshLocation(
    Bone / Socket Index=$Particles.PartitionParticles.Partition
    Bone / Socket Sampling Mode=::NewEnumerator5
    MaskSpawnGroup=false
    Mesh Sampling Type=::NewEnumerator6
    Skeletal Mesh=$System.SkeletalMesh
   )
   &AddVelocity(Velocity=&RandomRangeVector(Minimum=<-1500 0 0>, Maximum=<-355 0 300>)) {enabled=false}
   &AddVelocity(
    Cone Angle=&DistanceBetweenPositions(A=<0 0 20>, B=<0 0 0>)
    Cone Axis=<-1500 0 0>
    Inner Cone Angle=1
    Module.UseConeFalloff=false
    Module.UseVelDistribution=false
    Override Seed=false
    Velocity Mode=::NewEnumerator2
    Velocity Speed=1500
   ) {module=AddVelocity001}
   &ShapeLocation(MaskSpawnGroup=false, Override Seed=false, Sphere Radius=60, UseOwnerScale=true)
   &Color(Color=&RandomRangeLinearColor(Minimum=<rgba 0.97 0.61 0 1>, Maximum=<rgba 0.97 0.0788 0.0969 1>))
  ]
  particleUpdateScript [
   &ParticleState()
   &ScaleSpriteSize(Uniform Curve Sprite Scale=DataInterface() {propertyValues {Curve {keys [{tangentMode=RCTM_User, time=0.8, value=1}, {tangentMode=RCTM_User, time=1, value=0}]}}})
   &SpriteRotationRate(Rotation Rate=&RandomRangeFloat(Minimum=0.5, Maximum=360))
   &GravityForce(Gravity=&RandomRangeVector(
     Minimum=&VectorFromFloat(Value=&FloatFromCurve(FloatCurve=DataInterface() {propertyValues {Curve {keys [
           {arriveTangent=26315.7891, interpMode=RCIM_Cubic, time=0, value=1}
           {arriveTangent=-369.8653, interpMode=RCIM_Cubic, leaveTangent=-369.8653, tangentMode=RCTM_User, time=0.15, value=0}
           {arriveTangent=-12000.001, interpMode=RCIM_Cubic, leaveTangent=-12000.001, time=0.19, value=5000}
           {arriveTangent=163.0197, interpMode=RCIM_Cubic, leaveTangent=163.0197, tangentMode=RCTM_User, time=0.2, value=-600}
          ]}}}))
     Maximum=<0 0 -600>
    )) {module=GravityForce004}
   &ScaleVelocity(Velocity Scale=&VectorFromFloat(Value=&FloatFromCurve(FloatCurve=DataInterface() {propertyValues {Curve {keys [<at 0 1>,<at 0.1 0>]}}}))) {enabled=false}
   &GravityForce(Gravity=&RandomRangeVector(
     Minimum=&VectorFromFloat(Value=&FloatFromCurve(FloatCurve=DataInterface() {propertyValues {Curve {keys [
           <at 0 1 RCIM_Cubic>
           {arriveTangent=-1755, interpMode=RCIM_Cubic, leaveTangent=-1755, time=0.1, value=800}
           {arriveTangent=-7950.689, interpMode=RCIM_Cubic, leaveTangent=-7950.689, tangentMode=RCTM_User, time=0.2, value=-350}
          ]}}}))
     Maximum=<0 0 -350>
    )) {enabled=false}
   &GravityForce(Gravity=&RandomRangeVector(
     Minimum=&VectorFromFloat(Value=&FloatFromCurve(FloatCurve=DataInterface() {propertyValues {Curve {keys [
           <at 0 1 RCIM_Cubic>
           {arriveTangent=-369.8653, interpMode=RCIM_Cubic, leaveTangent=-369.8653, tangentMode=RCTM_User, time=0.18, value=0}
           {arriveTangent=-17500.0039, interpMode=RCIM_Cubic, leaveTangent=-17500.0039, time=0.19, value=5000}
           {arriveTangent=163.0197, interpMode=RCIM_Cubic, leaveTangent=163.0197, tangentMode=RCTM_User, time=0.2, value=-350}
          ]}}}))
     Maximum=<0 0 -600>
    )) {enabled=false, module=GravityForce003}
   &GravityForce(Gravity=&RandomRangeVector(
     Minimum=<0 0 -600>
     Maximum=&VectorFromFloat(Value=&FloatFromCurve(FloatCurve=DataInterface() {propertyValues {Curve {keys [<at 0 1>,<at 0.5 20>,<at 1 0>]}}}))
    )) {enabled=false, module=GravityForce002}
   &GravityForce(Gravity=&VectorFromFloat(Value=&FloatFromCurve(FloatCurve=DataInterface() {propertyValues {Curve {keys [{arriveTangent=2.0365, interpMode=RCIM_Cubic, leaveTangent=2.0365, tangentMode=RCTM_User, time=0, value=1}, <at 1 0.8>]}}}))) {module=GravityForce001}
   &Drag(Drag=0.1, Rotational Drag=0.3, Use Linear Drag=true, Use Rotational Drag=true) {enabled=false}
   &SolveForcesAndVelocity(Clamp Velocity=false, Limit Acceleration=false)
   &ScaleColor(Scale RGB=<1.2 1 1>, ScaleA=true, ScaleRGB=true)
   &RotateAroundPoint() {enabled=false}
   &DynamicMaterialParameters(
    Index 0 Param 1=&RandomRangeFloat(Minimum=0.75, Maximum=1)
    Index 0 Param 2=1
    Index 0 Param 3=&RandomRangeFloat(Minimum=0.15, Maximum=0.3)
    Index 0 Param 4=&RandomRangeFloat(Minimum=0.2, Maximum=0.3)
    Module.Index0Param0Write=true
    Module.Index0Param1Write=true
    Module.Index0Param2Write=true
    Module.Index0Param3Write=true
   ) {module=DynamicMaterialParameters001}
   &SubUVAnimation(
    "Sprite Renderer "=DataInterface() {propertyValues {SpriteRenderer=&self:NE_VC_Boost_Sprite_Parent003_9.NiagaraSpriteRendererProperties_0}}
    SubUV Lookup Index Scale=0.1
    UseEndFrame=false
    UseStartFrame=false
   )
  ]
  rendererClasses [&NiagaraSpriteRendererProperties]
  renderers [{rendererClass=&NiagaraSpriteRendererProperties}]
 }
}
summary {
 systemName=NS_Boost_Autumn
 emitters [
  {emitterName=NE_VC_Boost_Sprite_Parent001, rendererClasses [&NiagaraSpriteRendererProperties]}
  {
   emitterName=NE_VC_Boost_Sprite_Parent
   rendererClasses [&NiagaraRibbonRendererProperties, &NiagaraSpriteRendererProperties]
  }
  {emitterName=NE_VC_Boost_Sprite_Parent002, rendererClasses [&NiagaraSpriteRendererProperties]}
  {emitterName=NE_VC_Boost_Sprite_Parent003, rendererClasses [&NiagaraSpriteRendererProperties]}
 ]
 userVariables [
  User.bIsFrontendPreview {type=NiagaraBool}
  User.bIsPainted {type=NiagaraBool}
  User.ColorVariantID {type=NiagaraInt32}
  User.PaintedColor {type=LinearColor, description="", defaultValue {struct=LinearColor, value=<rgba 1 1 1 1>}}
  User.PaintedScalar {type=NiagaraFloat, defaultValue=1, description=""}
  User.PortCount {type=NiagaraInt32, defaultValue=1}
 ]
}
