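# Translucent, unlit, two-sided Niagara sprite material for leaf particles in the Autumn vehicle boost cosmetic.
# TextureSwitch blends a channel-packed single leaf with an autumn-leaf flipbook and its normal map.
# Transformed normals fake diffuse and half-vector specular lighting, with a static switch selecting the spec mask source.
# Particle color and dynamic parameters control brightness, specular and saturation before MF_FX_EmissiveAdjustments; a partly disconnected duplicate network also remains.
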Material /VehicleCosmetics/Boosts/Boost_Autumn/Shader/MT_Boost_Autumn
graph {
 properties {
  bAutomaticallySetUsageInEditor=false
  bUsedWithNiagaraSprites=true
  blendMode=BLEND_Translucent
  shadingModel=MSM_Unlit
  twoSided=true
  usedShadingModels=Unlit
 }
 nodes [
  ConstantBiasScale() {bias=0.05, scale=1.25}
  MaterialExpressionConstantClamp_102 = Saturate()
  MaterialExpressionVertexColor_4 = ParticleColor()
  n1 = TextureSample() {texture=/VehicleCosmetics/Boosts/Boost_Autumn/Texture/Leaf_T}
  MaterialExpressionConstantClamp_68 = Saturate(None=ConstantBiasScale(None=@n1.A) {bias=0, scale=1.5})
  n2 = TextureSample() {texture=/VehicleCosmetics/Boosts/Boost_Autumn/Texture/AutumnLeaves_FB_03_T}
  MaterialExpressionConstantClamp_69 = Saturate(None=ConstantBiasScale(None=@n2.A) {bias=-0.125, scale=1})
  n3 = ScalarParameter() {defaultValue=1, parameterName=TextureSwitch}
  MaterialExpressionConstantClamp_54 = Saturate(None=Multiply(A=LinearInterpolate(A=@MaterialExpressionConstantClamp_68, B=@MaterialExpressionConstantClamp_69, Alpha=@n3), B=Constant() {r=0.8}))
  MaterialExpressionConstantClamp_53 = Saturate(None=Multiply(A=@MaterialExpressionVertexColor_4.RGB, B=@MaterialExpressionConstantClamp_54))
  MaterialExpressionConstantClamp_88 = Saturate(None=Multiply())
  n4 = AppendVector()
  n5 = Constant()
  n6 = Multiply()
  n7 = Constant() {r=0.5}
  n8 = LinearInterpolate(A=Transform(), B=Transform())
  MaterialExpressionConstantClamp_98 = Saturate(None=ConstantBiasScale(None=DotProduct(A=@n8)))
  n9 = Constant() {r=3}
  MaterialExpressionConstantClamp_97 = Saturate(None=ConstantBiasScale(None=Power(Base=@MaterialExpressionConstantClamp_98, Exp=@n9)) {bias=0.135, scale=1})
  n10 = Add(A=LinearInterpolate(A=AppendVector(A=@n4, B=@n5), B=LinearInterpolate(B=@n6, Alpha=@n7)), B=Multiply(B=@MaterialExpressionConstantClamp_97))
  n11 = Normalize(VectorInput=Add(A=Transform(), B=CameraVectorWS()))
  n12 = TextureSample()
  MaterialExpressionConstantClamp_100 = Saturate(None=Multiply(A=@n12.RGB, B=Constant() {r=32}))
  n13 = Power(
   Base=ConstantBiasScale(None=DotProduct(B=@n11)) {bias=0.2, scale=0.8}
   Exp=StaticSwitchParameter(True=Multiply(B=@MaterialExpressionConstantClamp_100)) {parameterName=SpecMaskFlipBook}
  )
  n14 = StaticSwitchParameter(True=@n12.RGB) {parameterName=SpecMaskFlipBook}
  MaterialExpressionConstantClamp_101 = Saturate(None=ConstantBiasScale() {bias=0, scale=1.5})
  MaterialExpressionConstantClamp_99 = Saturate(None=Multiply(A=LinearInterpolate(A=@MaterialExpressionConstantClamp_101), B=Constant() {r=0.8}))
  MaterialExpressionConstantClamp_96 = Saturate(None=Multiply(A=Multiply(A=@n13, B=@n14), B=@MaterialExpressionConstantClamp_99))
  n15 = ConstantBiasScale(None=Desaturation(None=Add(A=@n10, B=@MaterialExpressionConstantClamp_96))) {bias=0, scale=0.95}
  MaterialExpressionVertexColor_6 = ParticleColor()
  n16 = Desaturation(None=@MaterialExpressionVertexColor_6.RGB)
  MaterialExpressionConstantClamp_89 = Saturate(None=Multiply(A=Multiply(A=@n15, B=@n16), B=DynamicParameter() {paramNames [Brightness, Specular, Saturation, DiffuseSaturation]}.Brightness))
  n17 = TextureSample() {texture=&Leaf_T}
  n18 = Add()
  MaterialExpressionConstantClamp_94 = Saturate(None=AppendVector(A=AppendVector(A=@n17.RGB, B=@n18), B=Add(A=@n17.RGB, B=@n17.RGB)))
  n19 = Constant3Vector() {constant=<rgba 0 0 1 0>}
  MaterialExpressionConstantClamp_95 = Saturate(None=OneMinus())
  n20 = LinearInterpolate(A=@MaterialExpressionConstantClamp_94, B=@n19, Alpha=@MaterialExpressionConstantClamp_95)
  MaterialExpressionConstantClamp_93 = Saturate(None=ConstantBiasScale(None=TextureSample() {texture=&AutumnLeaves_FB_03_T}.RGB) {bias=-0.125, scale=1})
  MaterialExpressionConstantClamp_92 = Saturate(None=OneMinus(None=@MaterialExpressionConstantClamp_93))
  n21 = LinearInterpolate(A=TextureSample().RGB, B=@n19, Alpha=@MaterialExpressionConstantClamp_92)
  n22 = ScalarParameter() {defaultValue=1, parameterName=TextureSwitch}
  MaterialExpressionConstantClamp_91 = Saturate(None=ConstantBiasScale(None=LinearInterpolate(A=@n20, B=@n21, Alpha=@n22)) {bias=0.05, scale=1})
  n23 = Constant() {r=1.25}
  MaterialExpressionConstantClamp_90 = Saturate(None=Multiply(A=Power(Base=@MaterialExpressionConstantClamp_91, Exp=@n23), B=Constant() {r=1.5}))
  Desaturation()
  n24 = LinearInterpolate(
   A=AppendVector(A=AppendVector(A=@n1.G, B=@n1.G), B=Constant())
   B=LinearInterpolate(A=@n2.RGB, B=Multiply(A=@n2.RGB, B=@n2.RGB), Alpha=Constant() {r=0.5})
  )
  n25 = Add(A=@n1.R, B=@n1.R)
  MaterialExpressionConstantClamp_67 = Saturate(None=AppendVector(A=AppendVector(A=@n1.R, B=@n25), B=Add(A=@n1.B, B=@n1.B)))
  n26 = TextureSample() {samplerType=SAMPLERTYPE_Normal, texture=/VehicleCosmetics/Boosts/Boost_Autumn/Texture/AutumnLeaves_FB_03_N_T}
  n27 = LinearInterpolate(A=Transform(None=@MaterialExpressionConstantClamp_67), B=Transform(None=@n26.RGB), Alpha=@n3)
  n28 = Constant3Vector() {constant=<rgba 0 0 1 0>}
  MaterialExpressionConstantClamp_64 = Saturate(None=ConstantBiasScale(None=DotProduct(A=@n27, B=@n28)))
  n29 = Constant() {r=3}
  MaterialExpressionConstantClamp_60 = Saturate(None=ConstantBiasScale(None=Power(Base=@MaterialExpressionConstantClamp_64, Exp=@n29)) {bias=0.135, scale=1})
  n30 = Multiply(A=@n24, B=@MaterialExpressionConstantClamp_60)
  n31 = DynamicParameter() {paramNames [Brightness, Specular, Saturation, DiffuseSaturation]}
  n32 = TextureSample() {texture=/VehicleCosmetics/Boosts/Boost_Autumn/Texture/Fruit_Seq_SpecMask}
  MaterialExpressionConstantClamp_63 = Saturate(None=Multiply(A=@n32.G, B=Constant() {r=32}))
  n33 = Normalize(VectorInput=Add(A=Transform(None=@n28), B=CameraVectorWS()))
  n34 = Power(
   Base=StaticSwitchParameter(True=@n31.Specular, False=Multiply(A=@n31.Specular, B=@MaterialExpressionConstantClamp_63)) {parameterName=SpecMaskFlipBook}
   Exp=ConstantBiasScale(None=DotProduct(A=@n27, B=@n33)) {bias=0.2, scale=0.8}
  )
  n35 = StaticSwitchParameter(True=@n2.R, False=@n32.G) {parameterName=SpecMaskFlipBook}
  MaterialExpressionConstantClamp_58 = Saturate(None=Multiply(A=Multiply(A=@n34, B=@n35), B=@MaterialExpressionConstantClamp_54))
  n36 = Desaturation(None=Add(A=Add(A=@n24, B=@n30), B=@MaterialExpressionConstantClamp_58), Fraction=@n31.DiffuseSaturation)
  n37 = ConstantBiasScale(None=@n36) {bias=0, scale=0.95}
  n38 = Desaturation(None=@MaterialExpressionVertexColor_4.RGB, Fraction=@n31.Saturation)
  MaterialExpressionConstantClamp_52 = Saturate(None=Multiply(A=Multiply(A=@n37, B=@n38), B=@n31.Brightness))
  MaterialFunctionCall(Emissive=@MaterialExpressionConstantClamp_52) {materialFunction=&MF_FX_EmissiveAdjustments}
  MaterialExpressionConstantClamp_66 = Saturate(None=OneMinus(None=@n1.R))
  n39 = LinearInterpolate(A=@MaterialExpressionConstantClamp_67, B=@n28, Alpha=@MaterialExpressionConstantClamp_66)
  MaterialExpressionConstantClamp_65 = Saturate(None=OneMinus(None=@MaterialExpressionConstantClamp_69))
  n40 = LinearInterpolate(A=@n26.RGB, B=@n28, Alpha=@MaterialExpressionConstantClamp_65)
  MaterialExpressionConstantClamp_48 = Saturate(None=ConstantBiasScale(None=LinearInterpolate(A=@n39, B=@n40, Alpha=@n3)) {bias=0.05, scale=1})
  n41 = Constant() {r=1.25}
  MaterialExpressionConstantClamp_47 = Saturate(None=Multiply(A=Power(Base=@MaterialExpressionConstantClamp_48, Exp=@n41), B=Constant() {r=1.5}))
  MaterialExpressionConstantClamp_59 = Saturate(None=ConstantBiasScale(None=Desaturation(None=@MaterialExpressionConstantClamp_47)) {bias=0.05, scale=1.25})
  Multiply(A=@MaterialExpressionConstantClamp_59, B=@n36)
  Multiply(A=@n24, B=Constant() {r=0.8})
  Multiply()
 ]
}
